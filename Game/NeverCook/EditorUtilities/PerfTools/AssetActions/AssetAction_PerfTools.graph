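# Editor asset action utility that dumps the selected assets' asset registry tags to AssetDataDump.csv in the project Saved folder.
# DumpSelectedAssetDataToCSV joins PerAssetHeaders lists (StaticMesh, Texture2D) into strings keyed by class path, then pairs each asset with its match.
# An embedded Python script reads those tags, or all tags for unlisted classes.
# It writes the CSV with merged columns and reveals it in Explorer.
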EditorUtilityBlueprint /Game/NeverCook/EditorUtilities/PerfTools/AssetActions/AssetAction_PerfTools
graphs [
 DumpSelectedAssetDataToCSV {nodes [
   n1 = GetSelectedAssetData(execute=@n4.then_1)
   > n2 = For Each Loop(Array:Array of Asset Data Structures=@n1) -> (Array Element:Asset Data Structure)
   n3 = GetTagValues!(execute, AssetPath:String="To String (Name)"(InName=Break Asset Data(AssetData=@n2))) -> (then, ReturnValue:Map of Names to Strings)
   > Keys!(execute, TargetMap:Map of Names to Strings=@n3) -> (then, Keys:Array of Names)
   DumpSelectedAssetDataToCSV() {is=title K2Node_FunctionEntry}
   > n4 = Sequence()
   n5 = Get PerAssetHeaders() -> (PerAssetHeaders:Map of Object Soft Class References to Asset Export Struct Structures)
   > n6 = Values!(execute, TargetMap:Map of Object Soft Class References to Asset Export Struct Structures=@n5) -> (then, Values:Array of Asset Export Struct Structures)
   > n7 = Keys(TargetMap:Map of Object Soft Class References to Asset Export Struct Structures=@n5) -> (Keys:Array of Object Soft Class References)
   > n8 = For Each Loop(Array:Array of Object Soft Class References=@n7) -> (Array Element:Object Soft Class Reference)
   > Set l_combinedString(l_combinedString:String) -> (Output_Get:String)
   > n9 = For Each Loop(Array:Array of Strings=Break Asset Export Struct!(AssetExportStruct:"Asset Export Struct Structure (by ref)"="Get (a ref)"(Array:Array of Asset Export Struct Structures=@n6, Dimension 1=@n8.Array Index) -> (Output:"Asset Export Struct Structure (by ref)")) -> (headers_7_#1:Array of Strings) {is=title K2Node_BreakStruct}) -> (Array Element:String)
   Add!(
    execute=@n9.Completed
    TargetMap:Map of Top Level Asset Path Structures to Strings=Get l_ClassPathToHeadStringMap() -> (l_ClassPathToHeadStringMap:Map of Top Level Asset Path Structures to Strings)
    Key:"Top Level Asset Path Structure (by ref)"=GetSoftClassTopLevelAssetPath!(SoftClassReference:Object Soft Class Reference=@n8) -> (ReturnValue:Top Level Asset Path Structure)
    Value:"String (by ref)"=Get l_combinedString() -> (l_combinedString:String)
   ) -> (then) {is=title K2Node_CallFunction}
   n10 = "integer > integer"(A=@n9.Array Index)
   Set l_combinedString(
    execute=@n9.LoopBody
    l_combinedString=Append!(A=Get l_combinedString(), B=Select(Option 0:String, Option 1:String=",", Index=@n10) -> (ReturnValue:String), C:String=@n9) -> (ReturnValue)
   )
   n11 = Keys(execute=@n2.LoopBody, TargetMap:Map of Top Level Asset Path Structures to Strings=Get l_ClassPathToHeadStringMap()) -> (Keys:Array of Top Level Asset Path Structures)
   > n12 = Find Item!(
    execute
    TargetArray:Array of Top Level Asset Path Structures=@n11
    ItemToFind:"Top Level Asset Path Structure (by ref)"=Break Asset Data(AssetData=@n2).AssetClassPath
   ) -> (then, ReturnValue)
   n13 = Values(execute=@n14.then, TargetMap:Map of Top Level Asset Path Structures to Strings=Get l_ClassPathToHeadStringMap()) -> (Values:Array of Strings)
   > Add(
    TargetMap:Map of Asset Data Structures to Strings=Get l_selectedDataMap() -> (l_selectedDataMap:Map of Asset Data Structures to Strings)
    Key:"Asset Data Structure (by ref)"=@n2
    Value="Get (a ref)"(Array:Array of Strings=@n13, Dimension 1=@n12) -> (Output:"String (by ref)")
   )
   Add(execute=@n14.else, TargetMap=Get l_selectedDataMap(), Key=@n2)
   Execute Python Script!(
    execute=@n2.Completed
    PythonScript="import unreal\r\nimport csv\r\nimport os\r\nimport subprocess\r\n\r\nfinal_csv_output = []\r\nall_headers = ['PackageName']\r\nfor asset_data in in_data_map:\r\n    header_string = in_data_map[asset_data]\r\n    package_name = str(asset_data.package_name)\r\n    header_array = []\r\n\r\n    if header_string != \"\":\r\n        header_array = header_string.split(',')\r\n    else:\r\n        ss = unreal.get_editor_subsystem(unreal.EditorAssetSubsystem)\r\n        for header in ss.get_tag_values(package_name):\r\n            header_array.append(str(header))\r\n\r\n    \r\n\r\n    asset_dict = {}\r\n    asset_dict['PackageName'] = package_name\r\n    \r\n    for header in header_array:\r\n        tag = asset_data.get_tag_value(header)\r\n        if tag is not None:\r\n            if header not in all_headers:\r\n                print(\"New header: \" + header)\r\n                all_headers.append(header)\r\n            asset_dict[header] = tag\r\n    \r\n    final_csv_output.append(asset_dict)\r\n\r\nif len(final_csv_output) > 0:\r\n    csv_file_path = unreal.Paths.project_saved_dir() + \"AssetDataDump.csv\"\r\n    \r\n    try:\r\n        with open(csv_file_path, mode='w', newline='') as csvfile:\r\n            writer = csv.DictWriter(csvfile, fieldnames=all_headers)\r\n            writer.writeheader()\r\n            for data in final_csv_output:\r\n                writer.writerow(data)\r\n        \r\n        finalPath = os.path.realpath(csv_file_path)\r\n        print(\"Saved asset data to: \" + finalPath)\r\n        subprocess.Popen(r'explorer /select,\"'+ finalPath + '\"')\r\n    except IOError:\r\n        print(\"I/O error\")"
    in_data_map:Map of Asset Data Structures to Strings=Get l_selectedDataMap()
   ) -> (then, ReturnValue) {inputs [in_data_map]}
   n14 = Branch(execute=@n12.then, Condition="integer > integer"(A=@n12, B=-1))
  ]}
 EventGraph {links [], nodes []}
]
variables [PerAssetHeaders {default {
   &StaticMesh {headers [
     CollisionComplexity
     SectionsWithCollision
     CollisionPrims
     DistanceFieldSize
     LODs
     Materials
     QualityLevelMinLOD
     Triangles
     Vertices
     UVChannels
     NaniteEnabled
     NaniteVertices
    ]}
   /Script/Engine.Texture2D {headers [MipGenSettings, CompressionSettings, Dimensions, TextureGroup, NeverStream, sRGB, Filter, Format, HasAlphaChannel, LODBias, MaxTextureSize]}
  }}]
